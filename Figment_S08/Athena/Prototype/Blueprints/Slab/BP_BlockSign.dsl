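; BP_BlockSign is a prototype Blueprint sign in the Slab folder that shows a player name as text.
; Its construction script sets the text of its TextRender component from the PlayerName variable, so the displayed name is applied when the actor is constructed.
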(fn ConstructionScript
  (Rendering|Components|TextRender|SetText (Variables|Default|GetTextRender) (Variables|Default|GetPlayerName)))
